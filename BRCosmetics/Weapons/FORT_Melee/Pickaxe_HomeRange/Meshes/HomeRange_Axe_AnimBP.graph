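# Post-input AnimBP for the HomeRange pickaxe that adds secondary motion to its dangling dyn_bullet bone chain.
# When bAnimDynamicsEnabled is true it uses chained AnimDynamics with separate tuning for frontend and other contexts; otherwise it uses one of two Trail controllers chosen by bIsOnPawn.
# The event graph derives inFrontend from the frontend and on-pawn flags, and raises trail RelaxationSpeedScale while falling, lowering it after a short delay otherwise.
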AnimBlueprint /BRCosmetics/Weapons/FORT_Melee/Pickaxe_HomeRange/Meshes/HomeRange_Axe_AnimBP
target_skeleton=&/HomeRange_Axe_Skeleton
graphs [
 AnimGraph {nodes [
   n1 = Anim Dynamics - Bone: dyn_bullet_01!(ComponentPose=Local To Component(LocalPose=Use cached pose 'InputPose'()), Alpha=1, GravityOverride:Vector="0,0,-980", ExternalForce="0,0,0") -> (Pose) {
    angularDampingOverride=0.98
    angularSpringConstant=2550
    bOverrideAngularDamping=true
    bUseGravityOverride=true
    gravityScale=0.35
    alphaBoolBlend {blendOption=Linear}
    boundBone {boneName=dyn_bullet_01}
    componentAppliedLinearAccClamp=<1000 1000 1000>
    componentLinearAccScale=<0.5 0.5 0.5>
    componentLinearVelScale=<0.5 0.5 0.5>
    gravityOverride {z=-980}
    physicsBodyDefinitions [{
      boundBone {boneName=dyn_bullet_01}
      boxExtents=<9 9 12>
      constraintSetup {angularLimitsMax=<0 115 0>, angularLimitsMin=<0 -115 0>, angularTarget=<0 0 0>}
      localJointOffset=<1 0 0>
     }]
    simSpaceSettings {maxAngularAcceleration=40, maxAngularVelocity=8, simSpaceAngularAlpha=1}
   }
   n2 = Anim Dynamics - Bone: dyn_bullet_02!(ComponentPose=@n1, Alpha=1, GravityOverride:Vector="0,0,-980", ExternalForce="0,0,0") -> (Pose) {
    angularDampingOverride=0.98
    angularSpringConstant=2550
    bOverrideAngularDamping=true
    bUseGravityOverride=true
    gravityScale=0.35
    alphaBoolBlend {blendOption=Linear}
    boundBone {boneName=dyn_bullet_02}
    componentAppliedLinearAccClamp=<1000 1000 1000>
    componentLinearAccScale=<0.5 0.5 0.5>
    componentLinearVelScale=<0.5 0.5 0.5>
    gravityOverride {z=-980}
    physicsBodyDefinitions [{
      boundBone {boneName=dyn_bullet_02}
      boxExtents=<9 9 12>
      constraintSetup {angularLimitsMax=<30 65 40>, angularLimitsMin=<-30 -65 -40>, angularTarget=<0 0 0>}
      localJointOffset=<1 0 0>
     }]
    simSpaceSettings {maxAngularAcceleration=40, maxAngularVelocity=8, simSpaceAngularAlpha=1}
   }
   n3 = Component To Local(ComponentPose=@n2)
   n4 = Anim Dynamics - Bone: dyn_bullet_01(ComponentPose=Local To Component(LocalPose=Use cached pose 'InputPose'()), Alpha=1, GravityOverride="0,0,-980", ExternalForce="0,0,0") {
    angularDampingOverride=0.98
    angularSpringConstant=2550
    bOverrideAngularDamping=true
    bUseGravityOverride=true
    gravityScale=0.35
    alphaBoolBlend {blendOption=Linear}
    boundBone {boneName=dyn_bullet_01}
    componentAppliedLinearAccClamp=<1000 1000 1000>
    componentLinearAccScale=<0.5 0.5 0.5>
    componentLinearVelScale=<0.5 0.5 0.5>
    gravityOverride {z=-980}
    physicsBodyDefinitions [{
      boundBone {boneName=dyn_bullet_01}
      boxExtents=<9 9 12>
      constraintSetup {angularLimitsMax=<0 115 0>, angularLimitsMin=<0 -20 0>, angularTarget=<0 0 0>}
      localJointOffset=<1 0 0>
     }]
   }
   n5 = Anim Dynamics - Bone: dyn_bullet_02(ComponentPose=@n4, Alpha=1, GravityOverride="0,0,-980", ExternalForce="0,0,0") {
    angularDampingOverride=0.98
    angularSpringConstant=2550
    bOverrideAngularDamping=true
    bUseGravityOverride=true
    gravityScale=0.35
    alphaBoolBlend {blendOption=Linear}
    boundBone {boneName=dyn_bullet_02}
    componentAppliedLinearAccClamp=<1000 1000 1000>
    componentLinearAccScale=<0.5 0.5 0.5>
    componentLinearVelScale=<0.5 0.5 0.5>
    gravityOverride {z=-980}
    physicsBodyDefinitions [{
      boundBone {boneName=dyn_bullet_02}
      boxExtents=<9 9 12>
      constraintSetup {angularLimitsMax=<30 65 40>, angularLimitsMin=<-30 -20 -40>, angularTarget=<0 0 0>}
      localJointOffset=<1 0 0>
     }]
   }
   n6 = Component To Local(ComponentPose=@n5)
   n7 = Get inFrontend() -> (inFrontend:bool)
   n8 = Trail controller - Bone: dyn_bullet_02(ComponentPose=Local To Component(LocalPose=Use cached pose 'InputPose'()), Alpha=1) {
    bActorSpaceFakeVel=true
    bLimitRotation=true
    maxDeltaTime=0.3
    fakeVelocity {x=30}
    rotationLimits [{limitMax=<60 60 60>, limitMin=<-90 -90 -90>}, {limitMax=<60 60 60>, limitMin=<-90 -90 -90>}]
    rotationOffsets [<0 0 0>, <0 0 0>]
    trailBone {boneName=dyn_bullet_02}
    trailRelaxationSpeed {editorCurveData {keys [<at 0 16 RCIM_Linear>,<at 1 14 RCIM_Linear>]}}
   }
   n9 = Component To Local(ComponentPose=@n8)
   n10 = Trail controller - Bone: dyn_bullet_02!(
    ComponentPose=Local To Component(LocalPose=Use cached pose 'InputPose'())
    Alpha=1
    RelaxationSpeedScale:float=Get RelaxationSpeedScale() -> (RelaxationSpeedScale:double)
   ) -> (Pose) {
    bActorSpaceFakeVel=true
    bEnableDebug=true
    bLimitRotation=true
    bShowTrailLocation=true
    maxDeltaTime=0.0333
    relaxationSpeedScale=6
    fakeVelocity {x=110}
    planarLimits [{drivingBone {boneName=dyn_bullet_01}, planeTransform {location=<-2 0 0>, rotation=<rot -50 180 -180>, scale=<1 1 1>}}]
    rotationLimits [{limitMax=<0 0 90>, limitMin=<0 0 -110>}, {limitMax=<0 0 0>, limitMin=<0 0 0>}]
    rotationOffsets [<0 0 0>, <0 0 0>]
    trailBone {boneName=dyn_bullet_02}
   }
   n11 = Component To Local(ComponentPose=@n10)
   n12 = Get bIsOnPawn() -> (bIsOnPawn:bool)
   Output Pose(Result=Blend Poses by bool(
     BlendPose_0=Blend Poses by bool(BlendPose_0=@n3, BlendPose_1=@n6, bActiveValue=@n7) {blendType=Linear}
     BlendPose_1=Blend Poses by bool(BlendPose_0=@n11, BlendPose_1=@n9, bActiveValue=@n12) {blendType=Linear}
     bActiveValue=Get bAnimDynamicsEnabled() -> (bAnimDynamicsEnabled:bool)
    ) {blendType=Linear}) {layerGroup=None}
   AnimGraphNode_SubInput_0 = Input Pose()
   Save cached pose 'InputPose'(Pose=@AnimGraphNode_SubInput_0) {cacheName=InputPose}
  ]}
 EventGraph {nodes [
   Event BlueprintUpdateAnimation()
   n1 = Get bIsOnPawn()
   > Set inFrontend(inFrontend:bool=AND Boolean(A=Get bIsInFrontEnd(), B=NOT Boolean(A=@n1))) -> (Output_Get:bool)
   > n2 = Branch(Condition=Get bIsFalling())
   > Set RelaxationSpeedScale(RelaxationSpeedScale:double=6) -> (Output_Get:double)
   Delay(execute=@n2.else, Duration=0.2)
   > Set RelaxationSpeedScale(RelaxationSpeedScale=0.7)
  ]}
]
variables [
 inFrontend {type=boolean}
 RelaxationSpeedScale {type=number}
]
